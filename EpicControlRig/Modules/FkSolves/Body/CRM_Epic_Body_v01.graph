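# Modular Control Rig body module that builds body controls and optionally drives the connected Body element from them.
# Construction nests Body under a Body Orbit control placed between both hips, or at Body without them, and adds a movable pivot proxy driving the orbit.
# Forward solve updates orbit visibility and the proxy, optionally locks scale, and projects Body from its control.
# Backward solve recovers the control through an inverse delta transform cached in metadata.
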ControlRigBlueprint /EpicControlRig/Modules/FkSolves/Body/CRM_Epic_Body_v01
graphs [
 ConstructionGraph {nodes [
   RigUnit_PrepareForExecution = &RigUnit_PrepareForExecution()
   RigVMFunction_Sequence = RigVMAggregateNode(ExecuteContext=@RigUnit_PrepareForExecution.ExecutePin)
   RigVMVariableNode(
    Value=/Script/ControlRig.RigUnit_ResolveConnector(Connector=(Type=Connector,Name="Body"), SkipSocket=False) {is=struct RigVMUnitNode}.Result
    ExecuteContext=@RigVMFunction_Sequence.A
   ) {variable=Body}
   n1 = RigVMVariableNode() {variable=Body Control Shape}
   Get Control Color From Metadata = RigVMFunctionReferenceNode(Color Override=@n1.Value.Color)
   n2 = &RigUnit_ResolveConnector(Connector=(Type=Connector,Name="Body"), SkipSocket=False)
   n3 = RigVMDispatchNode(
    False=(Rotation=(X=0,Y=0,Z=0,W=1),Translation=(X=0,Y=0,Z=0),Scale3D=(X=1,Y=1,Z=1))
    True=&RigUnit_GetTransform(bInitial=true, Item=@n2.Result).Transform
    Condition=@n2.bIsConnected
   ) {template="DISPATCH_RigVMDispatch_If(in Condition,in True,in False,out Result)"}
   n4 = &RigUnit_ResolveConnector(Connector=(Type=Connector,Name="Left Hip"), SkipSocket=False)
   n5 = &RigUnit_ResolveConnector(Connector=(Type=Connector,Name="Right Hip"), SkipSocket=False)
   Interpolate = /Script/RigVM.RigVMFunction_MathTransformLerp(
    A=(Rotation=(X=0,Y=0,Z=0,W=1),Translation=(X=0,Y=0,Z=0),Scale3D=(X=1,Y=1,Z=1))
    B=(Rotation=(X=0,Y=0,Z=0,W=1),Translation=(X=0,Y=0,Z=0),Scale3D=(X=1,Y=1,Z=1))
    T=0.5
    A.Translation=&RigUnit_GetTransform(bInitial=true, Item=@n4.Result).Transform.Translation
    B.Translation=&RigUnit_GetTransform(bInitial=true, Item=@n5.Result).Transform.Translation
   ) {is=struct RigVMUnitNode, template="Interpolate::Execute(in A,in B,in T,out Result)"}
   RigVMFunction_MathBoolAnd = &RigVMFunction_MathBoolAnd(A=@n4.bIsConnected, B=@n5.bIsConnected)
   Scale = /Script/RigVM.RigVMFunction_MathVectorScale(Factor=1.2, Value=@n1.Value.Transform.Scale3D) {is=struct RigVMUnitNode, template="Scale::Execute(in Value,in Factor,out Result)"}
   n6 = /Script/ControlRig.RigUnit_HierarchyAddControlTransform(
    InitialValue=(Rotation=(X=0,Y=0,Z=0,W=1),Translation=(X=0,Y=0,Z=0),Scale3D=(X=1,Y=1,Z=1))
    Name=Body Orbit
    OffsetSpace=LocalSpace
    Settings="(InitialSpace=LocalSpace,bUsePreferredRotationOrder=False,PreferredRotationOrder=YZX,Limits=(LimitTranslationX=(bMinimum=False,bMaximum=False),LimitTranslationY=(bMinimum=False,bMaximum=False),LimitTranslationZ=(bMinimum=False,bMaximum=False),LimitPitch=(b ...(truncated)"
    Settings.Shape.Color=@"Get Control Color From Metadata".Result
    OffsetTransform=/Script/RigVM.RigVMFunction_MathTransformMakeAbsolute(
     Parent=RigVMDispatchNode(False=@n3.Result, True=@Interpolate.Result, Condition=@RigVMFunction_MathBoolAnd.Result) {template="DISPATCH_RigVMDispatch_If(in Condition,in True,in False,out Result)"}.Result
     Local=RigVMVariableNode() {variable=Body Offset Transform}.Value
    ) {is=struct RigVMUnitNode, template="Make Absolute::Execute(in Local,in Parent,out Global)"}.Global
    Parent=&RigUnit_ResolveConnector(Connector=(Type=Connector,Name="Parent"), SkipSocket=False).Result
    Settings.Shape.Transform.Scale3D=@Scale.Result
    Settings.Shape.Name=@n1.Value.Name
    Settings.Shape.Transform.Rotation=@n1.Value.Transform.Rotation
    Settings.Shape.Transform.Translation=@n1.Value.Transform.Translation
    Settings.Shape.bVisible=@n1.Value.bVisible
    ExecutePin=@RigVMFunction_Sequence.B
   ) {
    is=struct RigVMUnitNode
    template="SpawnControl::Execute(in InitialValue,in Settings,in OffsetTransform,in OffsetSpace,in Parent,in Name,out Item)"
   }
   > RigVMVariableNode(Value=@n6.Item) {variable=Body Gimbal Control}
   > SpawnAnimationChannel = /Script/ControlRig.RigUnit_HierarchyAddAnimationChannelBool(InitialValue=False, LimitsEnabled=(), MaximumValue=True, MinimumValue=False, Name=Body Orbit Vis, Parent=@n6.Item) {
    is=struct RigVMUnitNode
    template="SpawnAnimationChannel::Execute(in InitialValue,in MinimumValue,in MaximumValue,in LimitsEnabled,in Parent,in Name,out Item)"
   }
   > RigVMVariableNode(Value=@SpawnAnimationChannel.Item) {variable=Body Gimbal Vis Channel}
   n7 = RigVMVariableNode() {variable=Body Control Shape}
   Get Control Color From Metadata_1 = RigVMFunctionReferenceNode(Color Override=@n7.Value.Color)
   > n8 = &RigUnit_HierarchyAddControlTransform(
    InitialValue=(Rotation=(X=0,Y=0,Z=0,W=1),Translation=(X=0,Y=0,Z=0),Scale3D=(X=1,Y=1,Z=1))
    Name=Body
    OffsetSpace=LocalSpace
    OffsetTransform=(Rotation=(X=0,Y=0,Z=0,W=1),Translation=(X=0,Y=0,Z=0),Scale3D=(X=1,Y=1,Z=1))
    Settings="(InitialSpace=LocalSpace,bUsePreferredRotationOrder=False,PreferredRotationOrder=YZX,Limits=(LimitTranslationX=(bMinimum=False,bMaximum=False),LimitTranslationY=(bMinimum=False,bMaximum=False),LimitTranslationZ=(bMinimum=False,bMaximum=False),LimitPitch=(b ...(truncated)"
    Settings.Shape.Color=@"Get Control Color From Metadata_1".Result
    Settings.Shape.Name=@n7.Value.Name
    Settings.Shape.Transform=@n7.Value.Transform
    Settings.Shape.bVisible=@n7.Value.bVisible
    Parent=@n6.Item
   ) {template="SpawnControl::Execute(in InitialValue,in Settings,in OffsetTransform,in OffsetSpace,in Parent,in Name,out Item)"}
   > RigVMVariableNode(Value=@n8.Item) {variable=Body Control}
   n9 = RigVMVariableNode() {variable=Body}
   > Set_Metadata = RigVMDispatchNode(
    Name=Body Delta Transform
    NameSpace=Self
    Value=&RigVMFunction_MathTransformMakeRelative(
     Global=&RigUnit_GetTransform(bInitial=true, Item=@n9.Value).Transform
     Parent=&RigUnit_GetTransform(bInitial=true, Item=@n8.Item).Transform
    ).Local
    Item=@n8.Item
   ) {template="DISPATCH_RigDispatch_SetMetadata(in Item,in Name,in NameSpace,in Value,out Success)"}
   > /Script/ControlRig.RigUnit_SetChannelHosts(
    Hosts=((Type=None,Name="None"))
    Channel=RigVMVariableNode() {variable=Body Gimbal Vis Channel}.Value
    Hosts.0=RigVMVariableNode() {variable=Body Control}.Value
   ) {is=struct RigVMUnitNode}
   Set_Module_Metadata_2_1 = RigVMDispatchNode(Name=Body Control, NameSpace=Root, Value=RigVMVariableNode() {variable=Body Control}.Value, ExecuteContext=@RigVMFunction_Sequence.D) {template="DISPATCH_RigDispatch_SetModuleMetadata(in Name,in NameSpace,in Value,out Success)"}
   Get_Module_Metadata = RigVMDispatchNode(Default=1, Name=Global Control Scale, NameSpace=Root) {template="DISPATCH_RigDispatch_GetModuleMetadata(in Name,in NameSpace,in Default,out Value,out Found)"}
   Make_Array = RigVMDispatchNode(
    Values=((Type=None,Name="None"),((),Type=None,Name="None"),((),Type=None,Name="None"))
    Values.0=RigVMVariableNode() {variable=Body Gimbal Control}.Value
    Values.2=RigVMVariableNode() {variable=Body Proxy Control}.Value
    Values.1=RigVMVariableNode() {variable=Body Control}.Value
   ) {template="DISPATCH_RigVMDispatch_ArrayMake(in Values,out Array)"}
   Set Control Scale = RigVMFunctionReferenceNode(
    Auto Scale Factor=1
    Global Scale=@Get_Module_Metadata.Value
    Controls=@Make_Array.Array
    Scale=RigVMVariableNode() {variable=Control Scale}.Value
    ExecuteContext=@RigVMFunction_Sequence.E
   )
   Construct Movable Proxy v01 = RigVMFunctionReferenceNode(
    Proxy Name=Body Movable Pivot
    Proxy Parent Control=(Type=Connector,Name="Parent")
    Proxy Shape Settings="(DisplayName=\"None\",InitialSpace=LocalSpace,bUsePreferredRotationOrder=False,PreferredRotationOrder=YZX,Limits=(LimitTranslationX=(bMinimum=False,bMaximum=False),LimitTranslationY=(bMinimum=False,bMaximum=False),LimitTranslationZ=(bMinimum=False,bMaximum=F ...(truncated)"
    Proxy Vis=false
    Proxy Snap To=RigVMVariableNode() {variable=Body}.Value
    Proxy Shape Settings.Shape=RigVMVariableNode() {variable=Body Proxy Control Shape}.Value
    Driven Control=RigVMVariableNode() {variable=Body Gimbal Control}.Value
    ExecuteContext=@RigVMFunction_Sequence.C
   )
   > RigVMVariableNode(Value=@"Construct Movable Proxy v01".Proxy Control) {variable=Body Proxy Control}
   > RigVMVariableNode(Value=@"Construct Movable Proxy v01".Proxy Buffer) {variable=Body Proxy Buffer}
   > RigVMVariableNode(Value=@"Construct Movable Proxy v01".Proxy Vis Channel) {variable=Body Proxy Vis Channel}
   > &RigUnit_SetChannelHosts(
    Hosts=((Type=None,Name="None"))
    Hosts.0=RigVMVariableNode() {variable=Body Control}.Value
    Channel=RigVMVariableNode() {variable=Body Proxy Vis Channel}.Value
   )
  ]}
 RigVMFunctionLibrary {links [], nodes []}
 RigVMFunction_Sequence_1_ContainedGraph {nodes [
   n1 = RigVMFunctionEntryNode()
   RigVMFunction_Sequence = &RigVMFunction_Sequence(ExecuteContext=@n1.ExecuteContext)
   n2 = &RigVMFunction_Sequence(ExecuteContext=@RigVMFunction_Sequence.B)
   n3 = &RigVMFunction_Sequence(ExecuteContext=@n2.B)
   RigVMFunctionReturnNode(A=@RigVMFunction_Sequence.A, B=@n2.A, C=@n3.A, D=@n3.B)
  ]}
 RigVMFunction_Sequence_1_ContainedGraph {nodes [
   n1 = RigVMFunctionEntryNode()
   RigVMFunction_Sequence = &RigVMFunction_Sequence(ExecuteContext=@n1.ExecuteContext)
   n2 = &RigVMFunction_Sequence(ExecuteContext=@RigVMFunction_Sequence.B)
   n3 = &RigVMFunction_Sequence(ExecuteContext=@n2.B)
   n4 = &RigVMFunction_Sequence(ExecuteContext=@n3.B)
   RigVMFunctionReturnNode(A=@RigVMFunction_Sequence.A, B=@n2.A, C=@n3.A, D=@n4.A, E=@n4.B)
  ]}
 RigVMModel {nodes [
   RigUnit_BeginExecution = &RigUnit_BeginExecution()
   RigVMFunction_Sequence = RigVMAggregateNode(ExecuteContext=@RigUnit_BeginExecution.ExecutePin)
   n1 = &RigUnit_ResolveConnector(Connector=(Type=Connector,Name="Body"), SkipSocket=False)
   RigVMFunction_MathBoolAnd = &RigVMFunction_MathBoolAnd(B=@n1.bIsConnected, A=RigVMVariableNode() {variable=Drive Body Joint}.Value)
   RigVMFunction_ControlFlowBranch = &RigVMFunction_ControlFlowBranch(Condition=@RigVMFunction_MathBoolAnd.Result, ExecuteContext=@RigVMFunction_Sequence.C)
   n2 = RigVMVariableNode() {variable=Body Control}
   &RigUnit_SetTransform(
    Weight=1
    bInitial=False
    Value=/Script/ControlRig.RigUnit_ProjectTransformToNewParent(bChildInitial=True, bNewParentInitial=False, bOldParentInitial=True, NewParent=@n2.Value, OldParent=@n2.Value, Child=@n1.Result) {is=struct RigVMUnitNode}.Transform
    ExecutePin=@RigVMFunction_ControlFlowBranch.True
    Item=@n1.Result
   )
   GetAnimationChannelFromItem = /Script/ControlRig.RigUnit_GetBoolAnimationChannelFromItem(bInitial=False, Item=RigVMVariableNode() {variable=Body Gimbal Vis Channel}.Value) {is=struct RigVMUnitNode, template="GetAnimationChannelFromItem::Execute(out Value,in Item,in bInitial)"}
   SetControlVisibility = /Script/ControlRig.RigUnit_SetControlVisibility(
    bVisible=@GetAnimationChannelFromItem.Value
    Item=RigVMVariableNode() {variable=Body Gimbal Control}.Value
    ExecutePin=@RigVMFunction_Sequence.A
   ) {is=struct RigVMUnitNode, template="SetControlVisibility::Execute(in Item,in Pattern,in bVisible)"}
   n3 = RigVMVariableNode() {variable=Body Gimbal Control}
   Forward Movable Proxy v01 = RigVMFunctionReferenceNode(
    Proxy Control=RigVMVariableNode() {variable=Body Proxy Control}.Value
    Proxy Buffer=RigVMVariableNode() {variable=Body Proxy Buffer}.Value
    Proxy Vis Channel=RigVMVariableNode() {variable=Body Proxy Vis Channel}.Value
    Driven Control=@n3.Value
    Proxy Snap To=@n3.Value
    ExecuteContext=@RigVMFunction_Sequence.D
   )
   n4 = &RigVMFunction_ControlFlowBranch(Condition=RigVMVariableNode() {variable=Lock Scale}.Value, ExecuteContext=@RigVMFunction_Sequence.B)
   Make_Array = RigVMDispatchNode(
    Values=((Type=None,Name="None"),((),Type=None,Name="None"),((),Type=None,Name="None"))
    Values.2=RigVMVariableNode() {variable=Body Proxy Control}.Value
    Values.1=RigVMVariableNode() {variable=Body Control}.Value
    Values.0=RigVMVariableNode() {variable=Body Gimbal Control}.Value
   ) {template="DISPATCH_RigVMDispatch_ArrayMake(in Values,out Array)"}
   For_Each = RigVMDispatchNode(Array=@Make_Array.Array, ExecuteContext=@n4.True) {template="DISPATCH_RigVMDispatch_ArrayIterator(in Array,out Element,out Index,out Count,out Ratio)"}
   > /Script/ControlRig.RigUnit_SetScale(Scale=(X=1,Y=1,Z=1), Space=LocalSpace, Weight=1, bInitial=False, bPropagateToChildren=True, Item=@For_Each.Element) {is=struct RigVMUnitNode}
  ]}
 RigVMModel Backwards Solve Graph {nodes [
   n1 = /Script/ControlRig.RigUnit_InverseExecution() {is=struct RigVMUnitNode}
   RigVMFunction_Sequence = &RigVMFunction_Sequence(ExecuteContext=@n1.ExecutePin)
   n2 = RigVMVariableNode() {variable=Body}
   n3 = RigVMVariableNode() {variable=Body Control}
   Get_Metadata = RigVMDispatchNode(
    Default=(Rotation=(X=0,Y=0,Z=0,W=1),Translation=(X=0,Y=0,Z=0),Scale3D=(X=1,Y=1,Z=1))
    Name=Body Delta Transform
    NameSpace=Self
    Item=@n3.Value
   ) {template="DISPATCH_RigDispatch_GetMetadata(in Item,in Name,in NameSpace,in Default,out Value,out Found)"}
   Inverse = /Script/RigVM.RigVMFunction_MathTransformInverse(Value=@Get_Metadata.Value) {is=struct RigVMUnitNode, template="Inverse::Execute(in Value,out Result)"}
   &RigUnit_SetTransform(
    Weight=1
    bInitial=False
    Value=&RigVMFunction_MathTransformMakeAbsolute(Parent=&RigUnit_GetTransform(bInitial=false, Item=@n2.Value).Transform, Local=@Inverse.Result) {template="Make Absolute::Execute(in Local,in Parent,out Global)"}.Global
    Item=@n3.Value
    ExecutePin=@RigVMFunction_Sequence.A
   )
  ]}
]
hierarchy [
 Parent {type=Connector, loc=<0 0 0>, rot=<0 0 0 1>}
 root {type=Socket, loc=<0 0 0>, rot=<0 0 0 1>}
 Body {type=Connector, loc=<0 0 0>, rot=<0 0 0 1>}
 Right Hip {type=Connector, loc=<0 0 0>, rot=<0 0 0 1>}
 Left Hip {type=Connector, loc=<0 0 0>, rot=<0 0 0 1>}
]
variables [
 Body {type=FRigElementKey, default=(Type=Connector,Name="")}
 Body Control {type=FRigElementKey, default=(Type=Bone,Name="")}
 Body Control Shape {
  type=FRigUnit_HierarchyAddControl_ShapeSettings
  default=(bVisible=True,Name="Hexagon_Thin",Color=(R=1,G=1,B=1,A=1),Transform=(Rotation=(X=0,Y=0,Z=0,W=1),Translation=(X=0,Y=0,Z=0),Scale3D=(X=7,Y=7,Z=7)))
 }
 Body Gimbal Control {type=FRigElementKey, default=(Type=Bone,Name="")}
 Body Gimbal Vis Channel {type=FRigElementKey, default=(Type=None,Name="")}
 Body Offset Transform {type=FTransform, default=(Rotation=(X=0,Y=0,Z=0,W=1),Translation=(X=0,Y=0,Z=0),Scale3D=(X=1,Y=1,Z=1))}
 Body Proxy Buffer {type=FRigElementKey, default=(Type=None,Name="")}
 Body Proxy Control {type=FRigElementKey, default=(Type=None,Name="")}
 Body Proxy Control Shape {
  type=FRigUnit_HierarchyAddControl_ShapeSettings
  default=(bVisible=True,Name="Sphere_Solid",Color=(R=1,G=0.0815,B=0.6989,A=1),Transform=(Rotation=(X=0,Y=0,Z=0,W=1),Translation=(X=0,Y=0,Z=0),Scale3D=(X=0.5,Y=0.5,Z=0.5)))
 }
 Body Proxy Vis Channel {type=FRigElementKey, default=(Type=None,Name="")}
 Control Scale {type=double, default=1}
 Drive Body Joint {type=bool, default=True}
 Lock Scale {type=bool, default=False}
]
