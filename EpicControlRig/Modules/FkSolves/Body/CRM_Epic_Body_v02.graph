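# Modular Control Rig body module that builds the body controls and drives the connected Body joint.
# Construction spawns Orbit, Body, optional Offset, Aim and movable-pivot proxy controls, placed at the hip midpoint when both hips connect, else at Body.
# Forward solve blends aim rotation and twist by channel weights, using a buffered rotation and auto-key during interaction, then poses the joint.
# Backward solve restores Body control from the joint via stored delta metadata.
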ControlRigBlueprint /EpicControlRig/Modules/FkSolves/Body/CRM_Epic_Body_v02
graphs [
 ConstructionGraph {nodes [
   RigUnit_PrepareForExecution = &RigUnit_PrepareForExecution()
   RigVMFunction_Sequence = RigVMAggregateNode(ExecuteContext=@RigUnit_PrepareForExecution.ExecutePin)
   RigVMVariableNode(
    Value=/Script/ControlRig.RigUnit_ResolveConnector(Connector=(Type=Connector,Name="Body"), SkipSocket=False) {is=struct RigVMUnitNode}.Result
    ExecuteContext=@RigVMFunction_Sequence.A
   ) {variable=Body}
   n1 = RigVMVariableNode() {variable=Body Control Shape}
   Get Control Color From Metadata = RigVMFunctionReferenceNode(Color Override=@n1.Value.Color)
   n2 = &RigUnit_ResolveConnector(Connector=(Type=Connector,Name="Body"), SkipSocket=False)
   n3 = RigVMDispatchNode(
    False=(Rotation=(X=0,Y=0,Z=0,W=1),Translation=(X=0,Y=0,Z=0),Scale3D=(X=1,Y=1,Z=1))
    True=&RigUnit_GetTransform(bInitial=true, Item=@n2.Result).Transform
    Condition=@n2.bIsConnected
   ) {template="DISPATCH_RigVMDispatch_If(in Condition,in True,in False,out Result)"}
   n4 = &RigUnit_ResolveConnector(Connector=(Type=Connector,Name="Left Hip"), SkipSocket=False)
   n5 = &RigUnit_ResolveConnector(Connector=(Type=Connector,Name="Right Hip"), SkipSocket=False)
   Interpolate = /Script/RigVM.RigVMFunction_MathTransformLerp(
    A=(Rotation=(X=0,Y=0,Z=0,W=1),Translation=(X=0,Y=0,Z=0),Scale3D=(X=1,Y=1,Z=1))
    B=(Rotation=(X=0,Y=0,Z=0,W=1),Translation=(X=0,Y=0,Z=0),Scale3D=(X=1,Y=1,Z=1))
    T=0.5
    A.Translation=&RigUnit_GetTransform(bInitial=true, Item=@n4.Result).Transform.Translation
    B.Translation=&RigUnit_GetTransform(bInitial=true, Item=@n5.Result).Transform.Translation
   ) {is=struct RigVMUnitNode, template="Interpolate::Execute(in A,in B,in T,out Result)"}
   RigVMFunction_MathBoolAnd = &RigVMFunction_MathBoolAnd(A=@n4.bIsConnected, B=@n5.bIsConnected)
   Scale = /Script/RigVM.RigVMFunction_MathVectorScale(Factor=1.2, Value=@n1.Value.Transform.Scale3D) {is=struct RigVMUnitNode, template="Scale::Execute(in Value,in Factor,out Result)"}
   n6 = /Script/ControlRig.RigUnit_HierarchyAddControlTransform(
    InitialValue=(Rotation=(X=0,Y=0,Z=0,W=1),Translation=(X=0,Y=0,Z=0),Scale3D=(X=1,Y=1,Z=1))
    Name=Body Orbit
    OffsetSpace=GlobalSpace
    Settings="(InitialSpace=LocalSpace,bUsePreferredRotationOrder=False,PreferredRotationOrder=YZX,Limits=(LimitTranslationX=(bMinimum=False,bMaximum=False),LimitTranslationY=(bMinimum=False,bMaximum=False),LimitTranslationZ=(bMinimum=False,bMaximum=False),LimitPitch=(b ...(truncated)"
    Settings.Shape.Color=@"Get Control Color From Metadata".Result
    OffsetTransform=/Script/RigVM.RigVMFunction_MathTransformMakeAbsolute(
     Parent=RigVMDispatchNode(False=@n3.Result, True=@Interpolate.Result, Condition=@RigVMFunction_MathBoolAnd.Result) {template="DISPATCH_RigVMDispatch_If(in Condition,in True,in False,out Result)"}.Result
     Local=RigVMVariableNode() {variable=Body Offset Transform}.Value
    ) {is=struct RigVMUnitNode, template="Make Absolute::Execute(in Local,in Parent,out Global)"}.Global
    Parent=&RigUnit_ResolveConnector(Connector=(Type=Connector,Name="Parent"), SkipSocket=False).Result
    Settings.Shape.Transform.Scale3D=@Scale.Result
    Settings.Shape.Name=@n1.Value.Name
    Settings.Shape.Transform.Rotation=@n1.Value.Transform.Rotation
    Settings.Shape.Transform.Translation=@n1.Value.Transform.Translation
    Settings.Shape.bVisible=@n1.Value.bVisible
    ExecutePin=@RigVMFunction_Sequence.B
   ) {
    is=struct RigVMUnitNode
    template="SpawnControl::Execute(in InitialValue,in Settings,in OffsetTransform,in OffsetSpace,in Parent,in Name,out Item)"
   }
   > RigVMVariableNode(Value=@n6.Item) {variable=Body Orbit Control}
   > SpawnAnimationChannel = /Script/ControlRig.RigUnit_HierarchyAddAnimationChannelBool(InitialValue=False, LimitsEnabled=(), MaximumValue=True, MinimumValue=False, Name=Body Orbit Vis, Parent=@n6.Item) {
    is=struct RigVMUnitNode
    template="SpawnAnimationChannel::Execute(in InitialValue,in MinimumValue,in MaximumValue,in LimitsEnabled,in Parent,in Name,out Item)"
   }
   > RigVMVariableNode(Value=@SpawnAnimationChannel.Item) {variable=Body Orbit Vis Channel}
   n7 = RigVMVariableNode() {variable=Body Control Shape}
   Get Control Color From Metadata_1 = RigVMFunctionReferenceNode(Color Override=@n7.Value.Color)
   > n8 = &RigUnit_HierarchyAddControlTransform(
    InitialValue=(Rotation=(X=0,Y=0,Z=0,W=1),Translation=(X=0,Y=0,Z=0),Scale3D=(X=1,Y=1,Z=1))
    Name=Body
    OffsetSpace=LocalSpace
    OffsetTransform=(Rotation=(X=0,Y=0,Z=0,W=1),Translation=(X=0,Y=0,Z=0),Scale3D=(X=1,Y=1,Z=1))
    Settings="(InitialSpace=LocalSpace,bUsePreferredRotationOrder=False,PreferredRotationOrder=YZX,Limits=(LimitTranslationX=(bMinimum=False,bMaximum=False),LimitTranslationY=(bMinimum=False,bMaximum=False),LimitTranslationZ=(bMinimum=False,bMaximum=False),LimitPitch=(b ...(truncated)"
    Settings.Shape.Color=@"Get Control Color From Metadata_1".Result
    Settings.Shape.Name=@n7.Value.Name
    Settings.Shape.Transform=@n7.Value.Transform
    Settings.Shape.bVisible=@n7.Value.bVisible
    Parent=@n6.Item
   ) {template="SpawnControl::Execute(in InitialValue,in Settings,in OffsetTransform,in OffsetSpace,in Parent,in Name,out Item)"}
   > RigVMVariableNode(Value=@n8.Item) {variable=Body Control}
   > RigVMVariableNode(Value=@n8.Item) {variable=Body Driver}
   n9 = RigVMVariableNode() {variable=Body}
   > Set_Metadata = RigVMDispatchNode(
    Name=Body Delta Transform
    NameSpace=Self
    Value=&RigVMFunction_MathTransformMakeRelative(
     Global=&RigUnit_GetTransform(bInitial=true, Item=@n9.Value).Transform
     Parent=&RigUnit_GetTransform(bInitial=true, Item=@n8.Item).Transform
    ).Local
    Item=@n8.Item
   ) {template="DISPATCH_RigDispatch_SetMetadata(in Item,in Name,in NameSpace,in Value,out Success)"}
   > n10 = /Script/ControlRig.RigUnit_SetChannelHosts(
    Hosts=((Type=None,Name="None"))
    Channel=RigVMVariableNode() {variable=Body Orbit Vis Channel}.Value
    Hosts.0=RigVMVariableNode() {variable=Body Control}.Value
   ) {is=struct RigVMUnitNode}
   RigVMFunction_ControlFlowBranch = &RigVMFunction_ControlFlowBranch(ExecuteContext=@n10.ExecutePin, Condition=RigVMVariableNode() {variable=Create Body Offset Control}.Value)
   n11 = RigVMVariableNode() {variable=Body Control}
   n12 = /Script/ControlRig.RigUnit_HierarchyGetShapeSettings(Item=@n11.Value) {is=struct RigVMUnitNode}
   Multiply = /Script/RigVM.RigVMFunction_MathColorMul(B=(R=0.5,G=0.5,B=0.5,A=1), A=@n12.Settings.Color) {is=struct RigVMUnitNode, template="Multiply::Execute(in A,in B,out Result)"}
   SpawnControl = &RigUnit_HierarchyAddControlTransform(
    InitialValue=(Rotation=(X=0,Y=0,Z=0,W=1),Translation=(X=0,Y=0,Z=0),Scale3D=(X=1,Y=1,Z=1))
    Name=Body Offset
    OffsetSpace=LocalSpace
    OffsetTransform=(Rotation=(X=0,Y=0,Z=0,W=1),Translation=(X=0,Y=0,Z=0),Scale3D=(X=1,Y=1,Z=1))
    Settings="(InitialSpace=LocalSpace,bUsePreferredRotationOrder=False,PreferredRotationOrder=YZX,Limits=(LimitTranslationX=(bMinimum=False,bMaximum=False),LimitTranslationY=(bMinimum=False,bMaximum=False),LimitTranslationZ=(bMinimum=False,bMaximum=False),LimitPitch=(b ...(truncated)"
    Settings.Shape.Name=@n12.Settings.Name
    Settings.Shape.Transform.Rotation=@n12.Settings.Transform.Rotation
    Settings.Shape.Transform.Translation=@n12.Settings.Transform.Translation
    Settings.Shape.bVisible=@n12.Settings.bVisible
    Settings.Shape.Color=@Multiply.Result
    ExecutePin=@RigVMFunction_ControlFlowBranch.True
    Settings.Shape.Transform.Scale3D=&RigVMFunction_MathVectorScale(Factor=0.85, Value=@n12.Settings.Transform.Scale3D) {template="Scale::Execute(in Value,in Factor,out Result)"}.Result
    Parent=@n11.Value
   ) {template="SpawnControl::Execute(in InitialValue,in Settings,in OffsetTransform,in OffsetSpace,in Parent,in Name,out Item)"}
   > RigVMVariableNode(Value=@SpawnControl.Item) {variable=Body Offset Control}
   > RigVMVariableNode(Value=@SpawnControl.Item) {variable=Body Driver}
   > n13 = &RigUnit_HierarchyAddAnimationChannelBool(InitialValue=true, LimitsEnabled=(), MaximumValue=True, MinimumValue=False, Name=Body Offset Vis, Parent=@SpawnControl.Item) {template="SpawnAnimationChannel::Execute(in InitialValue,in MinimumValue,in MaximumValue,in LimitsEnabled,in Parent,in Name,out Item)"}
   > RigVMVariableNode(Value=@n13.Item) {variable=Body Offset Vis Channel}
   > &RigUnit_SetChannelHosts(Hosts=((Type=None,Name="None")), Hosts.0=RigVMVariableNode() {variable=Body Control}.Value, Channel=@n13.Item)
   n14 = &RigUnit_HierarchyAddAnimationChannelBool(
    InitialValue=false
    LimitsEnabled=()
    MaximumValue=True
    MinimumValue=False
    Name=Body Aim Vis
    Parent=RigVMVariableNode() {variable=Body Aim Control}.Value
    ExecutePin=@RigVMFunction_Sequence.E
   ) {template="SpawnAnimationChannel::Execute(in InitialValue,in MinimumValue,in MaximumValue,in LimitsEnabled,in Parent,in Name,out Item)"}
   > RigVMVariableNode(Value=@n14.Item) {variable=Body Aim Vis Channel}
   > n15 = /Script/ControlRig.RigUnit_HierarchyAddAnimationChannelFloat(
    InitialValue=0
    LimitsEnabled=(Enabled=(bMinimum=True,bMaximum=True))
    MaximumValue=1
    MinimumValue=0
    Name=Aim Weight
    Parent=RigVMVariableNode() {variable=Body Aim Control}.Value
   ) {
    is=struct RigVMUnitNode
    template="SpawnAnimationChannel::Execute(in InitialValue,in MinimumValue,in MaximumValue,in LimitsEnabled,in Parent,in Name,out Item)"
   }
   > RigVMVariableNode(Value=@n15.Item) {variable=Body Aim Channel}
   > n16 = &RigUnit_HierarchyAddAnimationChannelFloat(
    InitialValue=0
    LimitsEnabled=(Enabled=(bMinimum=True,bMaximum=True))
    MaximumValue=1
    MinimumValue=0
    Name=Aim Twist
    Parent=RigVMVariableNode() {variable=Body Aim Control}.Value
   ) {template="SpawnAnimationChannel::Execute(in InitialValue,in MinimumValue,in MaximumValue,in LimitsEnabled,in Parent,in Name,out Item)"}
   > RigVMVariableNode(Value=@n16.Item) {variable=Body Aim Rotation Channel}
   n17 = RigVMVariableNode() {variable=Body Control}
   n18 = RigVMVariableNode() {variable=Body Orbit Control}
   > &RigUnit_SetChannelHosts(
    Hosts=((Type=None,Name="None"),(Type=None,Name="None"))
    Channel=RigVMVariableNode() {variable=Body Aim Vis Channel}.Value
    Hosts.0=@n17.Value
    Hosts.1=@n18.Value
   )
   > &RigUnit_SetChannelHosts(
    Hosts=((Type=None,Name="None"),(Type=None,Name="None"))
    Channel=RigVMVariableNode() {variable=Body Aim Channel}.Value
    Hosts.0=@n17.Value
    Hosts.1=@n18.Value
   )
   > &RigUnit_SetChannelHosts(
    Hosts=((Type=None,Name="None"),(Type=None,Name="None"))
    Channel=RigVMVariableNode() {variable=Body Aim Rotation Channel}.Value
    Hosts.0=@n17.Value
    Hosts.1=@n18.Value
   )
   Set_Module_Metadata_2_1 = RigVMDispatchNode(Name=Body Control, NameSpace=Root, Value=RigVMVariableNode() {variable=Body Control}.Value, ExecuteContext=@RigVMFunction_Sequence.F) {template="DISPATCH_RigDispatch_SetModuleMetadata(in Name,in NameSpace,in Value,out Success)"}
   n19 = RigVMVariableNode() {variable=Body Offset Control}
   n20 = &RigVMFunction_ControlFlowBranch(Condition=/Script/ControlRig.RigUnit_ItemExists(Item=@n19.Value) {is=struct RigVMUnitNode}.Exists, ExecuteContext=@RigVMFunction_Sequence.G)
   Get_Module_Metadata = RigVMDispatchNode(Default=1, Name=Global Control Scale, NameSpace=Root) {template="DISPATCH_RigDispatch_GetModuleMetadata(in Name,in NameSpace,in Default,out Value,out Found)"}
   Make_Array = RigVMDispatchNode(
    Values=((Type=None,Name="None"),((),Type=None,Name="None"),((),Type=None,Name="None"))
    Values.0=RigVMVariableNode() {variable=Body Orbit Control}.Value
    Values.2=RigVMVariableNode() {variable=Body Proxy Control}.Value
    Values.1=RigVMVariableNode() {variable=Body Control}.Value
   ) {template="DISPATCH_RigVMDispatch_ArrayMake(in Values,out Array)"}
   Set Control Scale = RigVMFunctionReferenceNode(
    Auto Scale Factor=1
    Global Scale=@Get_Module_Metadata.Value
    Controls=@Make_Array.Array
    ExecuteContext=@n20.Completed
    Scale=RigVMVariableNode() {variable=Control Scale}.Value
   )
   RigVMDispatchNode(Array=@Make_Array.Array, ExecuteContext=@n20.True, Element=@n19.Value) {template="DISPATCH_RigVMDispatch_ArrayAdd(io Array,in Element,out Index)"}
   Construct Movable Proxy v01 = RigVMFunctionReferenceNode(
    Proxy Name=Body Movable Pivot
    Proxy Parent Control=(Type=Connector,Name="Parent")
    Proxy Shape Settings="(DisplayName=\"None\",InitialSpace=LocalSpace,bUsePreferredRotationOrder=False,PreferredRotationOrder=YZX,Limits=(LimitTranslationX=(bMinimum=False,bMaximum=False),LimitTranslationY=(bMinimum=False,bMaximum=False),LimitTranslationZ=(bMinimum=False,bMaximum=F ...(truncated)"
    Proxy Vis=false
    Proxy Snap To=RigVMVariableNode() {variable=Body}.Value
    Proxy Shape Settings.Shape=RigVMVariableNode() {variable=Body Proxy Control Shape}.Value
    Driven Control=RigVMVariableNode() {variable=Body Orbit Control}.Value
    ExecuteContext=@RigVMFunction_Sequence.C
   )
   > RigVMVariableNode(Value=@"Construct Movable Proxy v01".Proxy Control) {variable=Body Proxy Control}
   > RigVMVariableNode(Value=@"Construct Movable Proxy v01".Proxy Buffer) {variable=Body Proxy Buffer}
   > RigVMVariableNode(Value=@"Construct Movable Proxy v01".Proxy Vis Channel) {variable=Body Proxy Vis Channel}
   > &RigUnit_SetChannelHosts(
    Hosts=((Type=None,Name="None"))
    Hosts.0=RigVMVariableNode() {variable=Body Control}.Value
    Channel=RigVMVariableNode() {variable=Body Proxy Vis Channel}.Value
   )
   n21 = &RigUnit_HierarchyAddControlTransform(
    InitialValue=(Rotation=(X=0,Y=0,Z=0,W=1),Translation=(X=0,Y=0,Z=0),Scale3D=(X=1,Y=1,Z=1))
    Name=Body Aim
    OffsetSpace=LocalSpace
    OffsetTransform=(Rotation=(X=0,Y=0,Z=0,W=1),Translation=(X=0,Y=0,Z=0),Scale3D=(X=1,Y=1,Z=1))
    Settings="(InitialSpace=LocalSpace,bUsePreferredRotationOrder=False,PreferredRotationOrder=YZX,Limits=(LimitTranslationX=(bMinimum=False,bMaximum=False),LimitTranslationY=(bMinimum=False,bMaximum=False),LimitTranslationZ=(bMinimum=False,bMaximum=False),LimitPitch=(b ...(truncated)"
    Parent=RigVMVariableNode() {variable=Body Orbit Control}.Value
    OffsetTransform.Translation=RigVMVariableNode() {variable=Body Aim Control Offset}.Value
    Settings.Shape=RigVMVariableNode() {variable=Aim Control Shape}.Value
    ExecutePin=@RigVMFunction_Sequence.D
   ) {template="SpawnControl::Execute(in InitialValue,in Settings,in OffsetTransform,in OffsetSpace,in Parent,in Name,out Item)"}
   > RigVMVariableNode(Value=@n21.Item) {variable=Body Aim Control}
   > /Script/ControlRig.RigUnit_AddAvailableSpaces(
    Spaces=((Key=(Type=Control,Name="None"),Label="Aim"))
    Spaces.0.Key=RigVMVariableNode() {variable=Body Aim Control}.Value
    Control=RigVMVariableNode() {variable=Body Control}.Value
   ) {is=struct RigVMUnitNode}
   n22 = &RigUnit_GetTransform(bInitial=False, Item=RigVMVariableNode() {variable=Body Control}.Value)
   > RigVMVariableNode(Value=@n22.Transform) {variable=Aim Buffer Transform}
   > RigVMVariableNode(Value=@n22.Transform.Rotation) {variable=Aim Rotate Buffer Quat}
  ]}
 New Function_ContainedGraph {nodes [
   RigVMFunctionEntryNode()
   > RigVMFunction_Sequence = RigVMAggregateNode()
   n1 = RigVMVariableNode(ExecuteContext=@RigVMFunction_Sequence.A, Value=RigVMVariableNode() {variable=Input Rotate Buffer Quat}.Value) {variable=Local Rotation Buffer Quat}
   n2 = /Script/ControlRig.RigUnit_IsInteracting() {is=struct RigVMUnitNode}
   n3 = RigVMVariableNode() {variable=Local Body Control}
   n4 = &RigVMFunction_ControlFlowBranch(
    Condition=&RigVMFunction_MathBoolAnd(
     B=RigVMDispatchNode(Array=@n2.Items, Element=@n3.Value) {template="DISPATCH_RigVMDispatch_ArrayFind(in Array,in Element,out Index,out Success)"}.Success
     A=@n2.bIsRotating
    ).Result
    ExecuteContext=@n1.ExecuteContext
   )
   n5 = RigVMVariableNode() {variable=Local Aim Control}
   n6 = RigVMVariableNode() {variable=Local Body Control}
   n7 = /Script/ControlRig.RigUnit_ProjectTransformToNewParent(bChildInitial=True, bNewParentInitial=False, bOldParentInitial=True, Child=@n5.Value, NewParent=@n6.Value, OldParent=@n6.Value) {is=struct RigVMUnitNode}
   RigVMFunction_MathQuaternionSwingTwist = &RigVMFunction_MathQuaternionSwingTwist(TwistAxis=(X=0,Y=0,Z=1), Input=@n7.Transform.Rotation)
   Multiply = /Script/RigVM.RigVMFunction_MathQuaternionMul(
    B=@RigVMFunction_MathQuaternionSwingTwist.Twist
    A=&RigVMFunction_MathQuaternionSwingTwist(TwistAxis=(X=0,Y=0,Z=1), Input=&RigUnit_GetTransform(Space=LocalSpace, bInitial=False, Item=@n5.Value).Transform.Rotation).Swing
   ) {is=struct RigVMUnitNode, template="Multiply::Execute(in A,in B,out Result)"}
   Set Transform_9 = &RigUnit_SetRotation(
    Space=GlobalSpace
    Weight=1
    bInitial=False
    Value=@Multiply.Result
    ExecutePin=@n4.True
    Item=RigVMVariableNode() {variable=Local Aim Control}.Value
   )
   > RigVMVariableNode(Value=@n7.Transform.Rotation) {variable=Local Rotation Buffer Quat}
   > SendEvent = /Script/ControlRig.RigUnit_SendEvent(
    Event=RequestAutoKey
    OffsetInSeconds=0
    bEnable=True
    bOnlyDuringInteraction=true
    Item=RigVMVariableNode() {variable=Local Body Control}.Value
   ) {is=struct RigVMUnitNode, template="SendEvent::Execute(in Event,in Item,in OffsetInSeconds,in bEnable,in bOnlyDuringInteraction)"}
   > &RigUnit_SendEvent(
    Event=RequestAutoKey
    OffsetInSeconds=0
    bEnable=True
    bOnlyDuringInteraction=True
    Item=RigVMVariableNode() {variable=Local Aim Control}.Value
   ) {template="SendEvent::Execute(in Event,in Item,in OffsetInSeconds,in bEnable,in bOnlyDuringInteraction)"}
   n8 = RigVMVariableNode() {variable=Local Body Control}
   n9 = &RigVMFunction_MathTransformMakeAbsolute(
    Local=(Rotation=(X=0,Y=0,Z=0,W=1),Translation=(X=0,Y=0,Z=0),Scale3D=(X=1,Y=1,Z=1))
    Parent=(Rotation=(X=0,Y=0,Z=0,W=1),Translation=(X=0,Y=0,Z=0),Scale3D=(X=1,Y=1,Z=1))
    Parent.Rotation=&RigUnit_GetTransform(bInitial=False, Item=/Script/ControlRig.RigUnit_HierarchyGetParent(bDefaultParent=True, Child=@n8.Value) {is=struct RigVMUnitNode}.Parent).Transform.Rotation
    Local.Rotation=RigVMVariableNode() {variable=Local Rotation Buffer Quat}.Value
   ) {template="Make Absolute::Execute(in Local,in Parent,out Global)"}
   n10 = RigVMVariableNode() {variable=Local Aim Control}
   Set Transform_3 = &RigUnit_SetRotation(
    Space=GlobalSpace
    Weight=1
    bInitial=false
    Value=&RigVMFunction_MathTransformLerp(
     A=(Rotation=(X=0,Y=0,Z=0,W=1),Translation=(X=0,Y=0,Z=0),Scale3D=(X=1,Y=1,Z=1))
     A.Rotation=@n9.Global.Rotation
     B=&RigUnit_ProjectTransformToNewParent(bChildInitial=True, bNewParentInitial=False, bOldParentInitial=True, Child=@n8.Value, NewParent=@n10.Value, OldParent=@n10.Value).Transform
     T=RigVMVariableNode() {variable=Interp Value}.Value
    ) {template="Interpolate::Execute(in A,in B,in T,out Result)"}.Result.Rotation
    Item=RigVMVariableNode() {variable=Local Body Control}.Value
    ExecutePin=@n4.False
   )
   RigVMFunctionReturnNode(ExecuteContext=@RigVMFunction_Sequence.B, Output Rotate Buffer Quat=RigVMVariableNode() {variable=Local Rotation Buffer Quat}.Value)
  ]}
 RigVMFunctionLibrary {nodes [Drive Aim and Body Rotation = RigVMCollapseNode(ExecuteContext=(), Input Rotate Buffer Quat=(), Local Aim Control=(Type=None,Name="None"), Local Body Control=(Type=None,Name="None"))]}
 RigVMFunction_Sequence_1_ContainedGraph {nodes [
   n1 = RigVMFunctionEntryNode()
   RigVMFunction_Sequence = &RigVMFunction_Sequence(ExecuteContext=@n1.ExecuteContext)
   n2 = &RigVMFunction_Sequence(ExecuteContext=@RigVMFunction_Sequence.B)
   n3 = &RigVMFunction_Sequence(ExecuteContext=@n2.B)
   n4 = &RigVMFunction_Sequence(ExecuteContext=@n3.B)
   RigVMFunctionReturnNode(A=@RigVMFunction_Sequence.A, B=@n2.A, C=@n3.A, D=@n4.A, E=@n4.B)
  ]}
 RigVMFunction_Sequence_1_ContainedGraph {nodes [
   n1 = RigVMFunctionEntryNode()
   RigVMFunction_Sequence = &RigVMFunction_Sequence(ExecuteContext=@n1.ExecuteContext)
   n2 = &RigVMFunction_Sequence(ExecuteContext=@RigVMFunction_Sequence.B)
   n3 = &RigVMFunction_Sequence(ExecuteContext=@n2.B)
   n4 = &RigVMFunction_Sequence(ExecuteContext=@n3.B)
   n5 = &RigVMFunction_Sequence(ExecuteContext=@n4.B)
   n6 = &RigVMFunction_Sequence(ExecuteContext=@n5.B)
   RigVMFunctionReturnNode(A=@RigVMFunction_Sequence.A, B=@n2.A, C=@n3.A, D=@n4.A, E=@n5.A, F=@n6.A, G=@n6.B)
  ]}
 RigVMFunction_Sequence_1_ContainedGraph {nodes [
   n1 = RigVMFunctionEntryNode()
   RigVMFunction_Sequence = &RigVMFunction_Sequence(ExecuteContext=@n1.ExecuteContext)
   n2 = &RigVMFunction_Sequence(ExecuteContext=@RigVMFunction_Sequence.B)
   RigVMFunctionReturnNode(A=@RigVMFunction_Sequence.A, B=@n2.A, C=@n2.B)
  ]}
 RigVMModel {nodes [
   RigUnit_BeginExecution = &RigUnit_BeginExecution()
   RigVMFunction_Sequence = RigVMAggregateNode(ExecuteContext=@RigUnit_BeginExecution.ExecutePin)
   n1 = &RigVMFunction_Sequence(ExecuteContext=@RigVMFunction_Sequence.C)
   n2 = RigVMVariableNode() {variable=Body Aim Channel}
   n3 = &RigVMFunction_ControlFlowBranch(
    Condition=/Script/RigVM.RigVMFunction_MathDoubleGreater(
     B=0
     A=/Script/ControlRig.RigUnit_GetFloatAnimationChannelFromItem(bInitial=False, Item=@n2.Value) {is=struct RigVMUnitNode, template="GetAnimationChannelFromItem::Execute(out Value,in Item,in bInitial)"}.Value
    ) {is=struct RigVMUnitNode, template="Greater::Execute(in A,in B,out Result)"}.Result
    ExecuteContext=@n1.A
   )
   Drive Aim and Body Rotation = RigVMFunctionReferenceNode(
    Interp Value=&RigUnit_GetFloatAnimationChannelFromItem(bInitial=False, Item=RigVMVariableNode() {variable=Body Aim Rotation Channel}.Value) {template="GetAnimationChannelFromItem::Execute(out Value,in Item,in bInitial)"}.Value
    ExecuteContext=@n3.True
    Local Aim Control=RigVMVariableNode() {variable=Body Aim Control}.Value
    Local Body Control=RigVMVariableNode() {variable=Body Control}.Value
    Input Rotate Buffer Quat=RigVMVariableNode() {variable=Aim Rotate Buffer Quat}.Value
   )
   > RigVMVariableNode(Value=@"Drive Aim and Body Rotation".Output Rotate Buffer Quat) {variable=Aim Rotate Buffer Quat}
   > RigVMVariableNode(
    Value=(Rotation=(X=0,Y=0,Z=0,W=1),Translation=(X=0,Y=0,Z=0),Scale3D=(X=1,Y=1,Z=1))
    Value.Rotation=@"Drive Aim and Body Rotation".Output Rotate Buffer Quat
   ) {variable=Aim Buffer Transform}
   RigVMVariableNode(
    Value=&RigUnit_GetTransform(Space=LocalSpace, bInitial=False, Item=RigVMVariableNode() {variable=Body Control}.Value).Transform.Rotation
    ExecuteContext=@n3.False
   ) {variable=Aim Rotate Buffer Quat}
   Greater = &RigVMFunction_MathDoubleGreater(
    B=0
    A=&RigUnit_GetFloatAnimationChannelFromItem(bInitial=False, Item=RigVMVariableNode() {variable=Body Aim Channel}.Value) {template="GetAnimationChannelFromItem::Execute(out Value,in Item,in bInitial)"}.Value
   ) {template="Greater::Execute(in A,in B,out Result)"}
   n4 = &RigVMFunction_ControlFlowBranch(Condition=@Greater.Result, ExecuteContext=@n1.B)
   RigVMVariableNode(
    Value=&RigUnit_GetTransform(Space=LocalSpace, bInitial=False, Item=RigVMVariableNode() {variable=Body Control}.Value).Transform
    ExecuteContext=@n4.False
   ) {variable=Aim Buffer Transform}
   n5 = &RigUnit_GetTransform(bInitial=False, Item=RigVMVariableNode() {variable=Body Aim Control}.Value)
   n6 = /Script/ControlRig.RigUnit_AimBoneMath(
    DebugSettings=(bEnabled=False,Scale=10,WorldOffset=(Rotation=(X=0,Y=0,Z=0,W=1),Translation=(X=0,Y=0,Z=0),Scale3D=(X=1,Y=1,Z=1)))
    Primary=(Weight=1,Axis=(X=0,Y=0,Z=1),Target=(X=1,Y=1,Z=1),Kind=Location,Space=(Type=Bone,Name="None"))
    Secondary=(Weight=1,Axis=(X=0,Y=0,Z=0),Target=(X=1,Y=0,Z=0),Kind=Direction,Space=(Type=Bone,Name="None"))
    Weight=1
    Primary.Target=@n5.Transform.Translation
    InputTransform=&RigUnit_GetTransform(bInitial=False, Item=RigVMVariableNode() {variable=Body Control}.Value).Transform
    Secondary.Target=/Script/RigVM.RigVMFunction_MathQuaternionRotateVector(Vector=(X=0,Y=0,Z=1), Transform=@n5.Transform.Rotation) {is=struct RigVMUnitNode, template="Rotate Vector::Execute(in Transform,in Vector,out Result)"}.Result
    Primary.Axis=RigVMVariableNode() {variable=Body Aim Axis}.Value
   ) {is=struct RigVMUnitNode}
   n7 = &RigUnit_HierarchyGetParent(bDefaultParent=True, Child=RigVMVariableNode() {variable=Body Control}.Value)
   Interpolate = &RigVMFunction_MathTransformLerp(
    B=@n6.Result
    T=&RigUnit_GetFloatAnimationChannelFromItem(bInitial=False, Item=RigVMVariableNode() {variable=Body Aim Channel}.Value) {template="GetAnimationChannelFromItem::Execute(out Value,in Item,in bInitial)"}.Value
    A=&RigVMFunction_MathTransformMakeAbsolute(
     Parent=&RigUnit_GetTransform(bInitial=False, Item=@n7.Parent).Transform
     Local=RigVMVariableNode() {variable=Aim Buffer Transform}.Value
    ) {template="Make Absolute::Execute(in Local,in Parent,out Global)"}.Global
   ) {template="Interpolate::Execute(in A,in B,in T,out Result)"}
   Set Transform_3 = &RigUnit_SetRotation(
    Space=GlobalSpace
    Weight=1
    bInitial=false
    Value=@Interpolate.Result.Rotation
    ExecutePin=@n4.True
    Item=RigVMVariableNode() {variable=Body Control}.Value
   )
   n8 = &RigUnit_ResolveConnector(Connector=(Type=Connector,Name="Body"), SkipSocket=False)
   RigVMFunction_MathBoolAnd = &RigVMFunction_MathBoolAnd(B=@n8.bIsConnected, A=RigVMVariableNode() {variable=Drive Body Joint}.Value)
   RigVMFunction_ControlFlowBranch = &RigVMFunction_ControlFlowBranch(Condition=@RigVMFunction_MathBoolAnd.Result, ExecuteContext=@RigVMFunction_Sequence.D)
   n9 = RigVMVariableNode() {variable=Body Driver}
   &RigUnit_SetTransform(
    Weight=1
    bInitial=False
    Value=&RigUnit_ProjectTransformToNewParent(bChildInitial=True, bNewParentInitial=False, bOldParentInitial=True, NewParent=@n9.Value, OldParent=@n9.Value, Child=@n8.Result).Transform
    ExecutePin=@RigVMFunction_ControlFlowBranch.True
    Item=@n8.Result
   )
   GetAnimationChannelFromItem = /Script/ControlRig.RigUnit_GetBoolAnimationChannelFromItem(bInitial=False, Item=RigVMVariableNode() {variable=Body Orbit Vis Channel}.Value) {is=struct RigVMUnitNode, template="GetAnimationChannelFromItem::Execute(out Value,in Item,in bInitial)"}
   SetControlVisibility = /Script/ControlRig.RigUnit_SetControlVisibility(
    bVisible=@GetAnimationChannelFromItem.Value
    Item=RigVMVariableNode() {variable=Body Orbit Control}.Value
    ExecutePin=@RigVMFunction_Sequence.A
   ) {is=struct RigVMUnitNode, template="SetControlVisibility::Execute(in Item,in Pattern,in bVisible)"}
   > n10 = &RigUnit_SetControlVisibility(
    bVisible=&RigUnit_GetBoolAnimationChannelFromItem(bInitial=False, Item=RigVMVariableNode() {variable=Body Aim Vis Channel}.Value) {template="GetAnimationChannelFromItem::Execute(out Value,in Item,in bInitial)"}.Value
    Item=RigVMVariableNode() {variable=Body Aim Control}.Value
   ) {template="SetControlVisibility::Execute(in Item,in Pattern,in bVisible)"}
   n11 = &RigVMFunction_ControlFlowBranch(ExecuteContext=@n10.ExecutePin, Condition=RigVMVariableNode() {variable=Create Body Offset Control}.Value)
   &RigUnit_SetControlVisibility(
    bVisible=&RigUnit_GetBoolAnimationChannelFromItem(bInitial=False, Item=RigVMVariableNode() {variable=Body Offset Vis Channel}.Value) {template="GetAnimationChannelFromItem::Execute(out Value,in Item,in bInitial)"}.Value
    ExecutePin=@n11.True
    Item=RigVMVariableNode() {variable=Body Offset Control}.Value
   ) {template="SetControlVisibility::Execute(in Item,in Pattern,in bVisible)"}
   n12 = RigVMVariableNode() {variable=Body Orbit Control}
   Forward Movable Proxy v01 = RigVMFunctionReferenceNode(
    Proxy Control=RigVMVariableNode() {variable=Body Proxy Control}.Value
    Proxy Buffer=RigVMVariableNode() {variable=Body Proxy Buffer}.Value
    Proxy Vis Channel=RigVMVariableNode() {variable=Body Proxy Vis Channel}.Value
    Driven Control=@n12.Value
    Proxy Snap To=@n12.Value
    ExecuteContext=@RigVMFunction_Sequence.E
   )
   n13 = &RigVMFunction_ControlFlowBranch(Condition=RigVMVariableNode() {variable=Lock Scale}.Value, ExecuteContext=@RigVMFunction_Sequence.B)
   n14 = &RigVMFunction_ControlFlowBranch(ExecuteContext=@n13.True, Condition=RigVMVariableNode() {variable=Create Body Offset Control}.Value)
   Make_Array = RigVMDispatchNode(
    Values=((Type=None,Name="None"),((),Type=None,Name="None"),((),Type=None,Name="None"))
    Values.2=RigVMVariableNode() {variable=Body Proxy Control}.Value
    Values.1=RigVMVariableNode() {variable=Body Control}.Value
    Values.0=RigVMVariableNode() {variable=Body Orbit Control}.Value
   ) {template="DISPATCH_RigVMDispatch_ArrayMake(in Values,out Array)"}
   For_Each = RigVMDispatchNode(Array=@Make_Array.Array, ExecuteContext=@n14.Completed) {template="DISPATCH_RigVMDispatch_ArrayIterator(in Array,out Element,out Index,out Count,out Ratio)"}
   > /Script/ControlRig.RigUnit_SetScale(Scale=(X=1,Y=1,Z=1), Space=LocalSpace, Weight=1, bInitial=False, bPropagateToChildren=True, Item=@For_Each.Element) {is=struct RigVMUnitNode}
   RigVMDispatchNode(ExecuteContext=@n14.True, Array=@Make_Array.Array, Element=RigVMVariableNode() {variable=Body Offset Control}.Value) {template="DISPATCH_RigVMDispatch_ArrayAdd(io Array,in Element,out Index)"}
  ]}
 RigVMModel Backwards Solve Graph {nodes [
   n1 = /Script/ControlRig.RigUnit_InverseExecution() {is=struct RigVMUnitNode}
   RigVMFunction_Sequence = &RigVMFunction_Sequence(ExecuteContext=@n1.ExecutePin)
   n2 = RigVMVariableNode() {variable=Body}
   n3 = RigVMVariableNode() {variable=Body Control}
   Get_Metadata = RigVMDispatchNode(
    Default=(Rotation=(X=0,Y=0,Z=0,W=1),Translation=(X=0,Y=0,Z=0),Scale3D=(X=1,Y=1,Z=1))
    Name=Body Delta Transform
    NameSpace=Self
    Item=@n3.Value
   ) {template="DISPATCH_RigDispatch_GetMetadata(in Item,in Name,in NameSpace,in Default,out Value,out Found)"}
   Inverse = /Script/RigVM.RigVMFunction_MathTransformInverse(Value=@Get_Metadata.Value) {is=struct RigVMUnitNode, template="Inverse::Execute(in Value,out Result)"}
   &RigUnit_SetTransform(
    Weight=1
    bInitial=False
    Value=&RigVMFunction_MathTransformMakeAbsolute(Parent=&RigUnit_GetTransform(bInitial=false, Item=@n2.Value).Transform, Local=@Inverse.Result) {template="Make Absolute::Execute(in Local,in Parent,out Global)"}.Global
    Item=@n3.Value
    ExecutePin=@RigVMFunction_Sequence.A
   )
   RigVMFunction_ControlFlowBranch = &RigVMFunction_ControlFlowBranch(Condition=RigVMVariableNode() {variable=Create Body Offset Control}.Value, ExecuteContext=@RigVMFunction_Sequence.B)
   &RigUnit_SetTransform(
    Space=LocalSpace
    Value=(Rotation=(X=0,Y=0,Z=0,W=1),Translation=(X=0,Y=0,Z=0),Scale3D=(X=1,Y=1,Z=1))
    Weight=1
    bInitial=False
    ExecutePin=@RigVMFunction_ControlFlowBranch.True
    Item=RigVMVariableNode() {variable=Body Offset Control}.Value
   )
  ]}
]
hierarchy [
 Parent {type=Connector, loc=<0 0 0>, rot=<0 0 0 1>}
 root {type=Socket, loc=<0 0 0>, rot=<0 0 0 1>}
 Body {type=Connector, loc=<0 0 0>, rot=<0 0 0 1>}
 Right Hip {type=Connector, loc=<0 0 0>, rot=<0 0 0 1>}
 Left Hip {type=Connector, loc=<0 0 0>, rot=<0 0 0 1>}
]
variables [
 Aim Buffer Transform {type=FTransform, default=(Rotation=(X=0,Y=0,Z=0,W=1),Translation=(X=0,Y=0,Z=0),Scale3D=(X=1,Y=1,Z=1))}
 Aim Control Shape {
  type=FRigUnit_HierarchyAddControl_ShapeSettings
  default=(bVisible=True,Name="ModularRigGizmoLibrary.RoundedTrapeze_ArrowTip_Thick",Color=(R=1,G=0.5,B=1,A=1),Transform=(Rotation=(X=0,Y=0,Z=0,W=1),Translation=(X=0,Y=0,Z=0),Scale3D=(X=5,Y=3.5,Z=3)))
 }
 Aim Rotate Buffer Quat {type=FQuat, default=(X=0,Y=0,Z=0,W=1)}
 Body {type=FRigElementKey, default=(Type=Connector,Name="")}
 Body Aim Axis {type=FVector, default=(X=0,Y=0,Z=1)}
 Body Aim Channel {type=FRigElementKey, default=(Type=None,Name="")}
 Body Aim Control {type=FRigElementKey, default=(Type=None,Name="")}
 Body Aim Control Offset {type=FVector, default=(X=0,Y=0,Z=38)}
 Body Aim Rotation Channel {type=FRigElementKey, default=(Type=None,Name="")}
 Body Aim Vis Channel {type=FRigElementKey, default=(Type=None,Name="")}
 Body Control {type=FRigElementKey, default=(Type=Bone,Name="")}
 Body Control Shape {
  type=FRigUnit_HierarchyAddControl_ShapeSettings
  default=(bVisible=True,Name="Hexagon_Thin",Color=(R=1,G=1,B=1,A=1),Transform=(Rotation=(X=0,Y=0,Z=0,W=1),Translation=(X=0,Y=0,Z=0),Scale3D=(X=7,Y=7,Z=7)))
 }
 Body Driver {type=FRigElementKey, default=(Type=Bone,Name="")}
 Body Gimbal Visibility Channel {type=FRigElementKey, default=(Type=None,Name="")}
 Body Offset Control {type=FRigElementKey, default=(Type=Bone,Name="")}
 Body Offset Transform {type=FTransform, default=(Rotation=(X=0,Y=0,Z=0,W=1),Translation=(X=0,Y=0,Z=0),Scale3D=(X=1,Y=1,Z=1))}
 Body Offset Vis Channel {type=FRigElementKey, default=(Type=Bone,Name="")}
 Body Orbit Control {type=FRigElementKey, default=(Type=Bone,Name="")}
 Body Orbit Vis Channel {type=FRigElementKey, default=(Type=None,Name="")}
 Body Proxy Buffer {type=FRigElementKey, default=(Type=None,Name="")}
 Body Proxy Control {type=FRigElementKey, default=(Type=None,Name="")}
 Body Proxy Control Shape {
  type=FRigUnit_HierarchyAddControl_ShapeSettings
  default=(bVisible=True,Name="Sphere_Solid",Color=(R=1,G=0.0815,B=0.6989,A=1),Transform=(Rotation=(X=0,Y=0,Z=0,W=1),Translation=(X=0,Y=0,Z=0),Scale3D=(X=0.5,Y=0.5,Z=0.5)))
 }
 Body Proxy Vis Channel {type=FRigElementKey, default=(Type=None,Name="")}
 Control Scale {type=double, default=1}
 Create Body Offset Control {type=bool, default=False}
 Drive Body Joint {type=bool, default=True}
 Lock Scale {type=bool, default=False}
]
